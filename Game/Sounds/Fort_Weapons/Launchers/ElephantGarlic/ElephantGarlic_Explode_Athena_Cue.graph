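# Sound cue for the ElephantGarlic launcher's explosion in Athena.
# After a short random delay, it mixes close, medium and distant layers, each with its own attenuation, sound class routing and a randomly picked wave variant.
# Medium and distant layers both draw on the Distant recordings with random pitch variation, at progressively lower mix volumes than the close layer.
# A concurrency override allows one instance at a time, stopping the oldest.
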SoundCue /Game/Sounds/Fort_Weapons/Launchers/ElephantGarlic/ElephantGarlic_Explode_Athena_Cue
root=Delay_0
nodes [
 {
  id=Attenuation_0
  children [Modulator_1]
  properties {
   attenuationSettings=&Fort_Weapon_Projectile_Distant_Mono_Athena_Att
   attenuationOverrides {falloffDistance=3000, attenuationShapeExtents {x=1000}}
  }
 }
 {
  id=Attenuation_4
  children [Random_1]
  properties {attenuationSettings=&Fort_Weapon_Projectile_Close_Athena_Att, attenuationOverrides {falloffDistance=1600}}
 }
 {
  id=Attenuation_5
  children [Modulator_5]
  properties {
   attenuationSettings=&Fort_Weapon_Projectile_Medium_Mono_Athena_Att
   attenuationOverrides {falloffDistance=3000, attenuationShapeExtents {x=1000}}
  }
 }
 {id=Delay_0, children [Mixer_4], properties {delayMax=0.1, delayMin=0.05}}
 {id=Mixer_4, children [SoundClass_3,SoundClass_2,SoundClass_0], properties {inputVolume [0.9,0.7,0.5]}}
 {id=Modulator_0, children [Attenuation_4]}
 {id=Modulator_1, children [Random_3], properties {pitchMax=1.1, pitchMin=0.9}}
 {id=Modulator_5, children [Random_2], properties {pitchMax=1.1, pitchMin=0.9}}
 {id=Random_1, children [WavePlayer_0,WavePlayer_5,WavePlayer_6,WavePlayer_7], properties {weights [1,1,1,1]}}
 {id=Random_2, children [WavePlayer_8, WavePlayer_9, WavePlayer_10, WavePlayer_11], properties {weights [1,1,1,1]}}
 {id=Random_3, children [WavePlayer_12, WavePlayer_13, WavePlayer_14, WavePlayer_15], properties {weights [1,1,1,1]}}
 {
  id=SoundClass_0
  children [Attenuation_0]
  properties {soundClassOverride=&Fort_WeaponRanged_Fire_Distant}
 }
 {
  id=SoundClass_2
  children [Attenuation_5]
  properties {soundClassOverride=&Fort_WeaponRanged_Fire_Distant}
 }
 {id=SoundClass_3, children [Modulator_0], properties {soundClassOverride=&Fort_WeaponRanged_Fire_3P_PSM}}
 {id=WavePlayer_0, properties {soundWaveAssetPtr=&/ElephantGarlic_Explode_Close_04}}
 {id=WavePlayer_10, properties {soundWaveAssetPtr=&/ElephantGarlic_Explode_Distant_03}}
 {id=WavePlayer_11, properties {soundWaveAssetPtr=&/ElephantGarlic_Explode_Distant_04}}
 {id=WavePlayer_12, properties {soundWaveAssetPtr=&/ElephantGarlic_Explode_Distant_01}}
 {id=WavePlayer_13, properties {soundWaveAssetPtr=&/ElephantGarlic_Explode_Distant_02}}
 {id=WavePlayer_14, properties {soundWaveAssetPtr=&ElephantGarlic_Explode_Distant_03}}
 {id=WavePlayer_15, properties {soundWaveAssetPtr=&ElephantGarlic_Explode_Distant_04}}
 {id=WavePlayer_5, properties {soundWaveAssetPtr=&/ElephantGarlic_Explode_Close_01}}
 {id=WavePlayer_6, properties {soundWaveAssetPtr=&/ElephantGarlic_Explode_Close_02}}
 {id=WavePlayer_7, properties {soundWaveAssetPtr=&/ElephantGarlic_Explode_Close_03}}
 {id=WavePlayer_8, properties {soundWaveAssetPtr=&ElephantGarlic_Explode_Distant_01}}
 {id=WavePlayer_9, properties {soundWaveAssetPtr=&ElephantGarlic_Explode_Distant_02}}
]
properties {bOverrideConcurrency=true, maxDistance=27500, volumeMultiplier=1, concurrencyOverrides {maxCount=1, resolutionRule=StopOldest}}
