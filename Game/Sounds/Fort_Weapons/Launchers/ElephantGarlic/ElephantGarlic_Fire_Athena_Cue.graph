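# Athena launcher fire sound cue for ElephantGarlic, classed as third-person ranged weapon fire.
# A mixer blends close, medium and half-volume distant layers, each with random wave selection and its own attenuation and sound class, with pitch variation on the distant layer.
# A quality switch gives the first level four variations per layer, while the other three levels share a mixer with fewer waves.
# Local and global gunfire concurrency sets govern playback.
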SoundCue /Game/Sounds/Fort_Weapons/Launchers/ElephantGarlic/ElephantGarlic_Fire_Athena_Cue
root=QualityLevel_0
nodes [
 {id=Attenuation_0, children [Modulator_0], properties {attenuationSettings=&Fort_Weapon_Projectile_Close_Athena_Att}}
 {
  id=Attenuation_1
  children [Random_8]
  properties {attenuationSettings=&Fort_Weapon_Projectile_Medium_Mono_Athena_Att}
 }
 {
  id=Attenuation_10
  children [Random_5]
  properties {attenuationSettings=&Fort_Weapon_Projectile_Medium_Mono_Athena_Att}
 }
 {
  id=Attenuation_11
  children [Modulator_7]
  properties {attenuationSettings=&Fort_Weapon_Projectile_Distant_Mono_Athena_Att}
 }
 {
  id=Attenuation_2
  children [Modulator_1]
  properties {attenuationSettings=&Fort_Weapon_Projectile_Distant_Mono_Athena_Att}
 }
 {id=Attenuation_9, children [Modulator_6], properties {attenuationSettings=&Fort_Weapon_Projectile_Close_Athena_Att}}
 {id=Mixer_0, children [SoundClass_3,SoundClass_4,SoundClass_0], properties {inputVolume [1,1,0.5]}}
 {id=Mixer_3, children [SoundClass_1,SoundClass_2,SoundClass_6], properties {inputVolume [1,1,0.5]}}
 {id=Modulator_0, children [Random_7], properties {pitchMax=1, pitchMin=1}}
 {id=Modulator_1, children [Random_9], properties {pitchMax=1.1}}
 {id=Modulator_6, children [Random_4], properties {pitchMax=1, pitchMin=1}}
 {id=Modulator_7, children [Random_6], properties {pitchMax=1.1}}
 {id=QualityLevel_0, children [Mixer_3,Mixer_0,Mixer_0,Mixer_0]}
 {id=Random_4, children [WavePlayer_15, WavePlayer_17, WavePlayer_18, WavePlayer_19], properties {weights [1,1,1,1]}}
 {id=Random_5, children [WavePlayer_21, WavePlayer_22, WavePlayer_23, WavePlayer_24], properties {weights [1,1,1,1]}}
 {id=Random_6, children [WavePlayer_25, WavePlayer_26, WavePlayer_27, WavePlayer_28], properties {weights [1,1,1,1]}}
 {id=Random_7, children [WavePlayer_29, WavePlayer_30, WavePlayer_31], properties {weights [1,1,1]}}
 {id=Random_8, children [WavePlayer_32, WavePlayer_33], properties {weights [1,1]}}
 {id=Random_9, children [WavePlayer_34, WavePlayer_35], properties {weights [1,1]}}
 {
  id=SoundClass_0
  children [Attenuation_2]
  properties {soundClassOverride=&Fort_WeaponRanged_Fire_Distant}
 }
 {
  id=SoundClass_1
  children [Attenuation_9]
  properties {soundClassOverride=&Fort_WeaponRanged_Fire_3P_PSM}
 }
 {
  id=SoundClass_2
  children [Attenuation_10]
  properties {soundClassOverride=&Fort_WeaponRanged_Fire_Distant}
 }
 {
  id=SoundClass_3
  children [Attenuation_0]
  properties {soundClassOverride=&Fort_WeaponRanged_Fire_3P_PSM}
 }
 {
  id=SoundClass_4
  children [Attenuation_1]
  properties {soundClassOverride=&Fort_WeaponRanged_Fire_Distant}
 }
 {
  id=SoundClass_6
  children [Attenuation_11]
  properties {soundClassOverride=&Fort_WeaponRanged_Fire_Distant}
 }
 {id=WavePlayer_15, properties {soundWaveAssetPtr=&/ElephantGarlic_Fire_Close_01}}
 {id=WavePlayer_17, properties {soundWaveAssetPtr=&/ElephantGarlic_Fire_Close_02}}
 {id=WavePlayer_18, properties {soundWaveAssetPtr=&/ElephantGarlic_Fire_Close_03}}
 {id=WavePlayer_19, properties {soundWaveAssetPtr=&/ElephantGarlic_Fire_Close_04}}
 {id=WavePlayer_21, properties {soundWaveAssetPtr=&/ElephantGarlic_Fire_Medium_01}}
 {id=WavePlayer_22, properties {soundWaveAssetPtr=&/ElephantGarlic_Fire_Medium_02}}
 {id=WavePlayer_23, properties {soundWaveAssetPtr=&/ElephantGarlic_Fire_Medium_03}}
 {id=WavePlayer_24, properties {soundWaveAssetPtr=&/ElephantGarlic_Fire_Medium_04}}
 {id=WavePlayer_25, properties {soundWaveAssetPtr=&/ElephantGarlic_Fire_Distant_01}}
 {id=WavePlayer_26, properties {soundWaveAssetPtr=&/ElephantGarlic_Fire_Distant_02}}
 {id=WavePlayer_27, properties {soundWaveAssetPtr=&/ElephantGarlic_Fire_Distant_03}}
 {id=WavePlayer_28, properties {soundWaveAssetPtr=&/ElephantGarlic_Fire_Distant_04}}
 {id=WavePlayer_29, properties {soundWaveAssetPtr=&ElephantGarlic_Fire_Close_02}}
 {id=WavePlayer_30, properties {soundWaveAssetPtr=&ElephantGarlic_Fire_Close_03}}
 {id=WavePlayer_31, properties {soundWaveAssetPtr=&ElephantGarlic_Fire_Close_04}}
 {id=WavePlayer_32, properties {soundWaveAssetPtr=&ElephantGarlic_Fire_Medium_03}}
 {id=WavePlayer_33, properties {soundWaveAssetPtr=&ElephantGarlic_Fire_Medium_04}}
 {id=WavePlayer_34, properties {soundWaveAssetPtr=&ElephantGarlic_Fire_Distant_03}}
 {id=WavePlayer_35, properties {soundWaveAssetPtr=&ElephantGarlic_Fire_Distant_04}}
]
properties {
 maxDistance=27500
 soundClassObject=&Fort_WeaponRanged_Fire_3P_PSM
 volumeMultiplier=0.9
 attenuationOverrides {falloffDistance=5000, attenuationShapeExtents {x=1000}}
 concurrencyOverrides {bLimitToOwner=true, maxCount=3, resolutionRule=PreventNew}
 concurrencySet [&Weapon_Gunfire_Local_Conc, &Weapon_Gunfire_Global_Conc]
}
